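; BP_PineSmell_Window_A_C is a window Blueprint in the PineSmell environment set of WildEstate.
; Its Construction Script runs the parent construction logic, then validates the glass provider components on itself through an SDP function.
; It adds no other logic of its own; any window behavior comes from the parent class and that validation call.
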(fn ConstructionScript
  (|CallParentFunction)
  (SDP|ValidateGlassProviderComponents self))
